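# Looping sound cue for the interior of the cylinder vortex at the Sky Island POI in Asteria.
# A single looping wave player plays the Vortex_Inside_Quad_Loop wave through a Modulator left at its default settings.
# The cue boosts volume, routes to the weather-related sound class and references a dedicated attenuation asset for its distance and spatial behavior.
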SoundCue /Game/Sounds/Asteria/POI/SkyIsland/Vortex/Vortex_Cylinder_Inside_Loop_Cue
root=Modulator_0
nodes [
 {id=Modulator_0, children [WavePlayer_0]}
 {id=WavePlayer_0, properties {bLooping=true, soundWaveAssetPtr=&/Vortex_Inside_Quad_Loop}}
]
properties {
 attenuationSettings=&/Vortex_Cylinder_Inside_Loop_Att
 maxDistance=9000
 soundClassObject=/Game/Sounds/FortSoundClassesAndModes/Fort_10_Weather_Related
 volumeMultiplier=1.5
 attenuationOverrides {bAttenuate=false, bSpatialize=false, falloffDistance=1350.8951, stereoSpread=5000, attenuationShapeExtents {x=2000}}
}
