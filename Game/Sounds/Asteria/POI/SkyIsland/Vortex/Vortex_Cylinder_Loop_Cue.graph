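# Sound cue for the Sky Island vortex cylinder ambience, routed to the storms sound class.
# Each mix blends outside, debris and distant loops with looped random thunder varied by delay and pitch, each layer separately attenuated.
# A quality switch gives the first slot fourteen thunder variants with quieter thunder and distant layers.
# The other three slots share a mix with three thunder variants excluded from branch culling.
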SoundCue /Game/Sounds/Asteria/POI/SkyIsland/Vortex/Vortex_Cylinder_Loop_Cue
root=QualityLevel_0
nodes [
 {
  id=Attenuation_0
  children [WavePlayer_32]
  properties {attenuationSettings=&/Vortex_Cylinder_Debris_Loop_Att}
 }
 {
  id=Attenuation_1
  children [WavePlayer_36]
  properties {attenuationSettings=&Vortex_Cylinder_Debris_Loop_Att}
 }
 {
  id=Attenuation_2
  children [WavePlayer_3]
  properties {attenuationSettings=&/Vortex_Cylinder_Loop_Att}
 }
 {
  id=Attenuation_3
  children [Looping_0]
  properties {attenuationSettings=&/Vortex_Med_Distant_Focus_Att}
 }
 {
  id=Attenuation_4
  children [WavePlayer_37]
  properties {attenuationSettings=&Vortex_Cylinder_Loop_Att}
 }
 {
  id=Attenuation_5
  children [Looping_1]
  properties {attenuationSettings=&Vortex_Med_Distant_Focus_Att}
 }
 {
  id=Attenuation_6
  children [WavePlayer_0]
  properties {attenuationSettings=&/Vortex_Distant_Focus_Att}
 }
 {
  id=Attenuation_7
  children [WavePlayer_1]
  properties {attenuationSettings=&Vortex_Distant_Focus_Att}
 }
 {id=Delay_0, children [Random_1], properties {delayMax=5}}
 {id=Delay_1, children [Random_2], properties {delayMax=5}}
 {id=Looping_0, children [Modulator_0]}
 {id=Looping_1, children [Modulator_1]}
 {id=Mixer_1, children [Attenuation_2, Attenuation_3, Attenuation_0, Attenuation_6], properties {inputVolume [1.75,0.25,1.5,0.3]}}
 {id=Mixer_6, children [Attenuation_4, Attenuation_5, Attenuation_1, Attenuation_7], properties {inputVolume [1.75,1,1.5,1.75]}}
 {id=Modulator_0, children [Delay_0], properties {pitchMax=1.1, pitchMin=0.93, volumeMax=1}}
 {id=Modulator_1, children [Delay_1], properties {pitchMax=1.1, pitchMin=0.93, volumeMax=1}}
 {id=QualityLevel_0, children [Mixer_1,Mixer_6,Mixer_6,Mixer_6]}
 {
  id=Random_1
  children [
   WavePlayer_5
   WavePlayer_18
   WavePlayer_19
   WavePlayer_20
   WavePlayer_21
   WavePlayer_22
   WavePlayer_24
   WavePlayer_25
   WavePlayer_26
   WavePlayer_27
   WavePlayer_28
   WavePlayer_29
   WavePlayer_30
   WavePlayer_31
  ]
  properties {weights [1,1,1,1,1,1,1,1,1,1,1,1,1,1]}
 }
 {id=Random_2, children [WavePlayer_33, WavePlayer_34, WavePlayer_35], properties {bShouldExcludeFromBranchCulling=true, weights [1,1,1]}}
 {id=WavePlayer_0, properties {bLooping=true, soundWaveAssetPtr=&/Vortex_Outside_Distant_Loop}}
 {id=WavePlayer_1, properties {bLooping=true, soundWaveAssetPtr=&Vortex_Outside_Distant_Loop}}
 {id=WavePlayer_18, properties {soundWaveAssetPtr=&/Vortex_Thumder_02}}
 {id=WavePlayer_19, properties {soundWaveAssetPtr=&/Vortex_Thumder_03}}
 {id=WavePlayer_20, properties {soundWaveAssetPtr=&/Vortex_Thumder_04}}
 {id=WavePlayer_21, properties {soundWaveAssetPtr=&/Vortex_Thumder_05}}
 {id=WavePlayer_22, properties {soundWaveAssetPtr=&/Vortex_Thumder_06}}
 {id=WavePlayer_24, properties {soundWaveAssetPtr=&/Vortex_Thumder_07}}
 {id=WavePlayer_25, properties {soundWaveAssetPtr=&/Vortex_Thumder_08}}
 {id=WavePlayer_26, properties {soundWaveAssetPtr=&/Vortex_Thumder_09}}
 {id=WavePlayer_27, properties {soundWaveAssetPtr=&/Vortex_Thumder_10}}
 {id=WavePlayer_28, properties {soundWaveAssetPtr=&/Vortex_Thumder_11}}
 {id=WavePlayer_29, properties {soundWaveAssetPtr=&/Vortex_Thumder_12}}
 {id=WavePlayer_3, properties {bLooping=true, soundWaveAssetPtr=&/Vortex_Outside_Loop}}
 {id=WavePlayer_30, properties {soundWaveAssetPtr=&/Vortex_Thumder_13}}
 {id=WavePlayer_31, properties {soundWaveAssetPtr=&/Vortex_Thumder_14}}
 {id=WavePlayer_32, properties {bLooping=true, soundWaveAssetPtr=&/Vortex_Outside_Loop_Debris}}
 {id=WavePlayer_33, properties {soundWaveAssetPtr=&Vortex_Thumder_06}}
 {id=WavePlayer_34, properties {soundWaveAssetPtr=&Vortex_Thumder_07}}
 {id=WavePlayer_35, properties {soundWaveAssetPtr=&Vortex_Thumder_08}}
 {id=WavePlayer_36, properties {bLooping=true, soundWaveAssetPtr=&Vortex_Outside_Loop_Debris}}
 {id=WavePlayer_37, properties {bLooping=true, soundWaveAssetPtr=&Vortex_Outside_Loop}}
 {id=WavePlayer_5, properties {soundWaveAssetPtr=&/Vortex_Thumder_01}}
]
properties {
 maxDistance=50000
 soundClassObject=/Game/Sounds/FortSoundClassesAndModes/Fort_Storms_PSM
 volumeMultiplier=1
 attenuationOverrides {
  bAttenuateWithLPF=true
  bEnableLogFrequencyScaling=true
  distanceAlgorithm=Logarithmic
  falloffDistance=10000
  lPFFrequencyAtMax=197.5057
  lPFFrequencyAtMin=20511.4102
  lPFRadiusMax=3789.3618
  lPFRadiusMin=1822.3112
  stereoSpread=2000
  attenuationShapeExtents {x=0}
 }
}
